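# Translucent, surface per-pixel lit material for the glass coils of the Tesla Coil vehicle cosmetic wheel, enabled for skeletal meshes and mesh deformers.
# The mask texture's red channel scales an emission color, while its green channel scales a glass color that multiplies the base color texture.
# Base color alpha times an opacity parameter provides transparency, alongside a normal map and scalar metallic and roughness parameters.
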Material /VehicleCosmetics/Wheels/TeslaCoil/Materials/M_Wheel_TeslaCoil_Coils
graph {
 properties {
  bAutomaticallySetUsageInEditor=false
  bUsedWithMeshDeformer=true
  bUsedWithSkeletalMesh=true
  blendMode=BLEND_Translucent
  translucencyLightingMode=TLM_SurfacePerPixelLighting
  usedShadingModels=Default Lit
 }
 nodes [
  n1 = TextureSampleParameter2D() {parameterName=Base Color, texture=/VehicleCosmetics/Wheels/TeslaCoil/Textures/T_Teslacoil_Glass_BC}
  Multiply(A=@n1.A, B=ScalarParameter() {defaultValue=1, parameterName=Opacity})
  n2 = TextureSampleParameter2D() {parameterName=Mask, texture=/VehicleCosmetics/Wheels/TeslaCoil/Textures/T_Teslacoil_Glass_M}
  Multiply(A=@n2.R, B=VectorParameter() {parameterName=Emission Color}.RGB)
  Multiply(A=Multiply(A=VectorParameter() {parameterName=Glass Color}.RGB, B=@n2.G), B=@n1.RGB)
  ScalarParameter() {parameterName=Metallic}
  ScalarParameter() {defaultValue=0.1, parameterName=Roughness}
  TextureSampleParameter2D() {parameterName=Normal, samplerType=SAMPLERTYPE_Normal, texture=/VehicleCosmetics/Wheels/TeslaCoil/Textures/T_Teslacoil_Glass_N}
 ]
}
