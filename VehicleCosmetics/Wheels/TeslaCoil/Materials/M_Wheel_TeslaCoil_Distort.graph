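# Translucent unlit material for the Tesla Coil cosmetic wheel, combining glowing energy effects with pixel-normal-offset refraction.
# Dot products of centred UVs build procedural circle and ring masks, with panning distortion textures warping the core's UVs.
# Named reroutes carry these masks, panning cloud noise, counter-rotating lightning, a flag-wave normal and vertex-colour green between sections.
# These sections build emissive, normal, opacity and refraction, with parameters for colours, brightness, distortion and coil-cylinder opacity.
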Material /VehicleCosmetics/Wheels/TeslaCoil/Materials/M_Wheel_TeslaCoil_Distort
graph {
 properties {
  bAutomaticallySetUsageInEditor=false
  bUsedWithMeshDeformer=true
  bUsedWithSkeletalMesh=true
  blendMode=BLEND_Translucent
  previewMesh=/VehicleCosmetics/Wheels/TeslaCoil/Meshes/SK_Wheel_Teslacoil
  refractionMethod=RM_PixelNormalOffset
  shadingModel=MSM_Unlit
  usedShadingModels=Unlit
 }
 nodes [
  n1 = TextureCoordinate()
  n2 = Multiply(A=Time()) {constB=0.1}
  n3 = TextureSample(UVs=Panner(Coordinate=@n1, Time=@n2) {speedX=-1, speedY=-1}) {samplerSource=SSM_Wrap_WorldGroupSettings, texture=/VehicleCosmetics/Wheels/_SharedTextures/FX/UVDistort_01_T}
  n4 = TextureSample(UVs=Panner(Coordinate=@n1, Time=@n2) {speedX=1, speedY=1}) {samplerSource=SSM_Wrap_WorldGroupSettings, texture=&UVDistort_01_T}
  NamedRerouteDeclaration(None=Add(
    A=Desaturation(None=@n3.RGB) {luminanceFactors=<rgba 0.3 0.59 0.11 1>}
    B=Desaturation(None=@n4.RGB) {luminanceFactors=<rgba 0.3 0.59 0.11 1>}
   )) {name=UV Dist Desat, nodeColor=<rgba 0.2422 1 0 1>}
  n5 = Multiply(A=Time()) {constB=0.05}
  n6 = TextureSample(UVs=Panner(Coordinate=TextureCoordinate() {uTiling=0.5, vTiling=0.5}, Time=@n5) {speedX=-0.5, speedY=-0.25}) {samplerSource=SSM_Wrap_WorldGroupSettings, texture=/VehicleCosmetics/Wheels/_SharedTextures/FX/Clouds_RGB_Soft_03_T}
  n7 = TextureSample(UVs=Panner(Coordinate=TextureCoordinate(), Time=@n5) {speedX=-0.25, speedY=0.5}) {samplerSource=SSM_Wrap_WorldGroupSettings, texture=&Clouds_RGB_Soft_03_T}
  NamedRerouteDeclaration(None=AppendVector(A=Add(A=@n6.G, B=@n7.G), B=Add(A=@n6.R, B=@n7.R))) {name=Noise Distort, nodeColor=<rgba 1 0 0.5625 1>}
  n8 = TextureCoordinate()
  n9 = Add(A=@n8) {constB=-0.5}
  n10 = DotProduct(A=@n9, B=@n9)
  n11 = Subtract(A=@n10) {constB=0.075}
  n12 = OneMinus(None=@n10)
  n13 = Clamp(None=Multiply(A=Subtract(A=@n12) {constB=0.8}) {constB=1.5}) {desc=Circle Small}
  n14 = Multiply(A=Clamp(None=Multiply(A=@n11) {constB=2}) {desc=Circle Medium}, B=@n13)
  NamedRerouteDeclaration(None=Multiply(A=@n14) {constB=20}) {name=Ring Soft 20, nodeColor=<rgba 0 1 0.5547 1>}
  NamedRerouteDeclaration(None=Multiply(A=@n14) {constB=8}) {name=Ring Soft 8, nodeColor=<rgba 0 1 0.5547 1>}
  n15 = AppendVector(A=@n3.G, B=@n4.G)
  n16 = AppendVector(A=@n3.B, B=@n4.B)
  n17 = Multiply(A=ScalarParameter() {defaultValue=0.01, parameterName=UVDistort}, B=Add(A=AppendVector(A=@n3.R, B=@n4.R), B=Multiply(A=@n15, B=@n16)))
  n18 = Clamp(None=Multiply(A=Subtract(A=@n12) {constB=0.92}) {constB=20}) {desc=Circle Medium}
  n19 = Add(A=Add(A=Multiply(A=@n17, B=@n18), B=@n8)) {constB=-0.5}
  n20 = Subtract(A=OneMinus(None=DotProduct(A=@n19, B=@n19))) {constB=0.997}
  NamedRerouteDeclaration(None=Clamp(None=Multiply(A=@n20) {constB=50}) {desc=Core}) {name=Core, nodeColor=<rgba 0 0.0859 1 1>}
  n21 = Subtract(A=@n12) {constB=0.982}
  n22 = Subtract(A=@n18, B=Clamp(None=Multiply(A=@n21) {constB=20}) {desc=Circle Small}) {desc=Thick Ring}
  n23 = Clamp(None=ConstantBiasScale(None=Clamp(None=@n22) {desc=Circle Medium}) {bias=-0.85, scale=1}) {desc=Thick Ring}
  NamedRerouteDeclaration(None=Multiply(A=@n23, B=@n23)) {name=Thick Ring, nodeColor=<rgba 0 0.25 1 1>}
  n24 = Subtract(A=@n12) {constB=0.85}
  n25 = Subtract(A=@n12) {constB=0.9}
  n26 = Subtract(A=Clamp(None=Multiply(A=@n24) {constB=20}), B=Clamp(None=Multiply(A=@n25) {constB=20})) {desc=Thin Ring}
  n27 = Clamp(None=ConstantBiasScale(None=Clamp(None=@n26) {desc=Circle Medium}) {bias=-0.85, scale=4}) {desc=Thin Ring}
  NamedRerouteDeclaration(None=Multiply(A=@n27, B=@n27)) {name=Thin Ring, nodeColor=<rgba 0.5703 1 0 1>}
  n28 = Desaturation(None=NamedRerouteUsage()) {luminanceFactors=<rgba 0.3 0.59 0.11 1>}
  n29 = Multiply(A=NamedRerouteUsage(), B=@n28)
  n30 = ScalarParameter() {defaultValue=20, parameterName=CoreBrightness}
  n31 = NamedRerouteUsage()
  n32 = Multiply(A=Add(A=Multiply(A=@n28, B=@n31), B=@n31), B=ScalarParameter() {defaultValue=60, parameterName=ThickRingBrightness})
  n33 = NamedRerouteUsage()
  n34 = Multiply(A=ScalarParameter() {defaultValue=5, parameterName=ThinRingBrightness}, B=NamedRerouteUsage())
  n35 = NamedRerouteUsage()
  n36 = Add(A=Multiply(A=Multiply(A=@n29, B=@n30), B=NamedRerouteUsage()), B=Add(A=Multiply(A=@n32, B=@n33), B=Multiply(A=@n34, B=@n35)))
  n37 = Add(A=Multiply(A=NamedRerouteUsage(), B=NamedRerouteUsage()), B=NamedRerouteUsage())
  n38 = Multiply(A=NamedRerouteUsage(), B=NamedRerouteUsage())
  n39 = NamedRerouteUsage()
  n40 = Add(A=NamedRerouteUsage(), B=Multiply(A=@n38, B=@n39))
  n41 = NamedRerouteUsage()
  NamedRerouteDeclaration(None=Add(A=Add(A=@n36, B=@n37), B=Multiply(A=@n40, B=@n41))) {name=Emissive, nodeColor=<rgba 1 0 0.8906 1>}
  NamedRerouteDeclaration(None=Add(A=@n3.G, B=@n4.R)) {name=Dist N X, nodeColor=<rgba 0 1 0.7891 1>}
  NamedRerouteDeclaration(None=Add(A=@n3.R, B=@n4.G)) {name=Dist N Y, nodeColor=<rgba 0 1 0.7891 1>}
  n42 = Multiply(A=Constant3Vector() {constant=<rgba 0.25 0 1 0>}, B=NamedRerouteUsage())
  n43 = Constant3Vector() {constant=<rgba 0 0 1 0>}
  n44 = Multiply(A=@n43, B=NamedRerouteUsage())
  n45 = TextureCoordinate() {uTiling=0.5, vTiling=0.5}
  n46 = Multiply(
   A=Multiply(A=Add(A=@n42, B=@n44)) {constB=0.5}
   B=Multiply(A=TextureSample(UVs=@n45) {samplerType=SAMPLERTYPE_Normal, texture=/VehicleCosmetics/Wheels/_SharedTextures/FX/FlagWave_NORM}.RGB) {constB=0.5}
  )
  NamedRerouteDeclaration(None=LinearInterpolate(A=@n43, B=Clamp(None=@n46), Alpha=NamedRerouteUsage())) {name=Normal, nodeColor=<rgba 0.1484 0 1 1>}
  NamedRerouteDeclaration(None=Multiply(A=Multiply(A=@n38, B=@n39)) {constB=0.5}) {name=Soft Mask between rings, nodeColor=<rgba 0 0.7656 1 1>}
  n47 = TextureCoordinate()
  n48 = Multiply(A=Time()) {constB=0.25}
  n49 = TextureSample(UVs=Rotator(Coordinate=@n47, Time=@n48) {centerY=0.25, speed=-0.25}) {samplerSource=SSM_Wrap_WorldGroupSettings, texture=/VehicleCosmetics/Wheels/_SharedTextures/FX/Lightning_RGB_Random_T}
  n50 = TextureSample(UVs=Rotator(Coordinate=@n47, Time=@n48) {centerX=0.25, speed=0.35}) {samplerSource=SSM_Wrap_WorldGroupSettings, texture=&Lightning_RGB_Random_T}
  n51 = Multiply(A=@n49.G, B=@n50.R)
  n52 = Multiply(A=@n49.R, B=@n50.B)
  NamedRerouteDeclaration(None=Add(A=Add(A=@n51, B=@n52), B=Multiply(A=@n49.B, B=@n50.G))) {name=Lightnings Rotation, nodeColor=<rgba 1 0 0.5625 1>}
  n53 = VectorParameter() {parameterName=Emissive Color, defaultValue=<rgba 0.9911 0.1195 0.0252 1>}
  NamedRerouteDeclaration(None=@n53.RGB) {name=Emissive Color, nodeColor=<rgba 0.8281 0 1 1>}
  NamedRerouteDeclaration(None=Power(Base=VectorParameter() {parameterName=Custom Color, defaultValue=<rgba 0.3712 0.6867 1 1>}.RGB)) {name=Custom Color, nodeColor=<rgba 0.8281 0 1 1>}
  n54 = NamedRerouteUsage()
  NamedRerouteDeclaration(None=Multiply(A=Power(Base=@n53.RGB) {constExponent=4}, B=Multiply(A=@n54) {constB=0.1}) {constB=0.1}) {name=Emissive Color Back, nodeColor=<rgba 0.8281 0 1 1>}
  n55 = NamedRerouteUsage()
  n56 = OneMinus(None=Add(A=@n23, B=Add(A=@n55, B=@n27)))
  NamedRerouteDeclaration(None=Clamp(None=ConstantBiasScale(None=@n56) {bias=-0.9, scale=60}) {desc=Circles Mask}) {name=Circles Mask, nodeColor=<rgba 1 0.7969 0 1>}
  NamedRerouteDeclaration(None=@n13) {name=50% Gray Ring Big, nodeColor=<rgba 0.4766 0 1 1>}
  n57 = Subtract(A=@n12) {constB=0.82}
  NamedRerouteDeclaration(None=Clamp(None=Multiply(A=@n57) {constB=6}) {desc=Circle Big Soft}) {name=Circle Big Soft, nodeColor=<rgba 1 0 0.1641 1>}
  NamedRerouteDeclaration(None=VertexColor().G) {name=VC_G, nodeColor=<rgba 0 1 0 1>}
  n58 = NamedRerouteUsage()
  n59 = Multiply(A=@n58, B=@n58)
  n60 = ScalarParameter() {defaultValue=1, parameterName=Opacity}
  n61 = Multiply(A=@n60, B=Add(A=NamedRerouteUsage(), B=NamedRerouteUsage()))
  n62 = Add(A=Add(A=Multiply(A=@n59) {constB=0.1}, B=Clamp(None=@n61)), B=NamedRerouteUsage())
  n63 = NamedRerouteUsage()
  n64 = LinearInterpolate(A=Clamp(None=Add(A=@n62, B=@n63)), B=ScalarParameter() {defaultValue=1, parameterName=CoilCylinderOpacity}, Alpha=NamedRerouteUsage())
  NamedRerouteDeclaration(None=Clamp(None=Multiply(A=@n64, B=@n60))) {name=Opacity, nodeColor=<rgba 0.8047 1 0 1>}
  n65 = Multiply(A=NamedRerouteUsage(), B=NamedRerouteUsage())
  n66 = ScalarParameter() {defaultValue=1, parameterName=Distotion}
  NamedRerouteDeclaration(None=LinearInterpolate(A=Multiply(A=@n65, B=@n66), Alpha=NamedRerouteUsage()) {constB=0}) {name=Refraction, nodeColor=<rgba 0.875 0 1 1>}
  NamedRerouteUsage()
  NamedRerouteUsage()
  NamedRerouteUsage()
  NamedRerouteUsage()
 ]
}
