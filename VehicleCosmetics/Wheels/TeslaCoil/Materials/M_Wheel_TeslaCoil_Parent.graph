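# Parent material for Tesla Coil vehicle wheels, layering animated electric emissive over MF_WheelParent_Simple, then applying damage and de-rez transition functions.
# An RGBA mask separates two flat colors, a lightning-colored region and an LOD black mask.
# Lightning multiplies two counter-panned RGB textures, shaped by scrolling noise and an energy parameter.
# Zero-tiled noise panners create UV-independent, thresholded flicker that gates the green and blue lightning channels.
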Material /VehicleCosmetics/Wheels/TeslaCoil/Materials/M_Wheel_TeslaCoil_Parent
graph {
 properties {
  bAutomaticallySetUsageInEditor=false
  bUseMaterialAttributes=true
  bUsedWithMeshDeformer=true
  bUsedWithSkeletalMesh=true
  pixelDepthOffsetMode=PDOM_Legacy
  usedShadingModels=Default Lit
 }
 nodes [
  n1 = TextureCoordinate() {uTiling=0, vTiling=0}
  n2 = Panner(Coordinate=@n1) {speedX=1, speedY=0.325}
  n3 = Panner(Coordinate=@n1) {speedX=-1, speedY=-0.15}
  n4 = AppendVector(
   A=AppendVector(
    A=TextureSample(UVs=@n2) {samplerSource=SSM_Wrap_WorldGroupSettings, texture=/VehicleCosmetics/Wheels/_SharedTextures/FX/PanRGBA_02_T}.G
    B=TextureSample(UVs=@n3) {samplerSource=SSM_Wrap_WorldGroupSettings, texture=&PanRGBA_02_T}.G
   )
   B=TextureSample(UVs=Panner(Coordinate=@n1) {speedX=-0.5, speedY=0.15}) {samplerSource=SSM_Wrap_WorldGroupSettings, texture=&PanRGBA_02_T}.G
  )
  n5 = Clamp(None=Multiply(A=Subtract(A=@n4) {constB=0.2}) {constB=35})
  ComponentMask(None=@n5) {r=true}
  n6 = GetMaterialAttributes(None=MaterialFunctionCall() {materialFunction=&MF_WheelParent_Simple}.MaterialAttributes) {attributeGetTypes [#1, #2, #3]}
  n7 = NamedRerouteUsage()
  n8 = SetMaterialAttributes!(
   MaterialAttributes=@n6.MaterialAttributes
   Base Color=LinearInterpolate(A=@n6.Base Color, Alpha=@n7) {constB=0}
   Metallic=LinearInterpolate(A=@n6.Metallic, Alpha=@n7)
   Roughness=LinearInterpolate(A=@n6.Roughness, Alpha=@n7)
   Emissive Color=NamedRerouteUsage()
  ) {attributeSetTypes [
    #1
    #2
    #3
    #4
   ]}
  MaterialFunctionCall(In=MaterialFunctionCall(Material=@n8) {materialFunction=&MF_BaseVehicle_Damage}.Result) {materialFunction=&MF_VehicleDeRez_Transition}
  n9 = TextureSampleParameter2D() {parameterName=Emissive Masks, samplerType=SAMPLERTYPE_LinearColor, texture=/VehicleCosmetics/Wheels/TeslaCoil/Textures/T_Teslacoil_Rim_E}
  n10 = Multiply(A=VectorParameter() {parameterName=Emissive_ColorA}.RGB, B=@n9.R)
  n11 = Multiply(A=VectorParameter() {parameterName=Emissive_ColorB}.RGB, B=@n9.G)
  NamedRerouteDeclaration(None=LinearInterpolate(A=LinearInterpolate(A=@n10, B=@n11, Alpha=@n9.G), B=NamedRerouteUsage(), Alpha=@n9.B)) {name=Emissive, nodeColor=<rgba 1 0 0.7969 1>}
  NamedRerouteDeclaration(None=@n9.B) {name=Emissive Masks B, nodeColor=<rgba 1 0 0.3984 1>}
  n12 = TextureSample(UVs=Panner(Coordinate=TextureCoordinate() {uTiling=3}) {speedX=-0.15, speedY=-0.75}) {samplerSource=SSM_Wrap_WorldGroupSettings, texture=/VehicleCosmetics/Wheels/_SharedTextures/FX/Lightning_RGB_Random_T}
  n13 = TextureSample(UVs=Panner(Coordinate=TextureCoordinate() {uTiling=3, vTiling=0.5}) {speedX=0.15, speedY=-0.75}) {samplerSource=SSM_Wrap_WorldGroupSettings, texture=&Lightning_RGB_Random_T}
  n14 = Add(
   A=Multiply(A=ComponentMask(None=@n5) {g=true}, B=Multiply(A=@n12.G, B=@n13.G))
   B=Multiply(A=ComponentMask(None=@n5) {b=true}, B=Multiply(A=@n12.B, B=@n13.B))
  )
  n15 = Multiply(A=@n12.R, B=@n13.R)
  n16 = NamedRerouteUsage()
  n17 = ScalarParameter() {defaultValue=5, parameterName=FlowEnergyEmissive}
  n18 = Multiply(A=Multiply(A=Add(A=@n15, B=@n14), B=@n16), B=@n17)
  n19 = TextureSample(UVs=Panner(Coordinate=TextureCoordinate() {vTiling=0.5}) {speedY=-0.3}) {samplerSource=SSM_Wrap_WorldGroupSettings, texture=&PanRGBA_02_T}
  n20 = Multiply(A=@n16, B=@n19.G)
  n21 = Multiply(A=@n19.B, B=@n16)
  n22 = Multiply(A=Add(A=Add(A=@n18, B=@n20), B=@n21), B=@n21)
  n23 = SquareRoot(None=VectorParameter() {parameterName=Emissive_ColorC, defaultValue=<rgba 1 0 0.0802 1>}.RGB)
  NamedRerouteDeclaration(None=Multiply(A=Multiply(A=@n22, B=@n17), B=Clamp(None=@n23))) {name=Out Emissive Color C, nodeColor=<rgba 0 0.7187 1 1>}
  NamedRerouteDeclaration(None=@n9.A) {name=Black Mask for LOD's, nodeColor=<rgba 1 0.8203 0 1>}
 ]
}
